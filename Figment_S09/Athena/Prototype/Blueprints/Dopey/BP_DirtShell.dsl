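; Dirt shell vote manager that shares health for a set of destructible props across mesh network servers, picking the first depleted as winner.
; Edge nodes scale prop damage and forward it through per-index damage component calls.
; The root subtracts health, latches VoteWinner once, and persists health and tuning to mesh metadata for restoration after restarts.
; OnRep handlers apply health to the props; winner, tick and playlist-setup handlers are empty stubs.
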(event Custom|SelectServerDamage
  (params (Damage :type Integer) (Index :type Integer))
  (bind _bp_dirtshellmeshnetwork (Variables|Default|GetBP_DirtShellMeshNetwork))
  ;[1 1 1] Kill me.
  (switch int Index
    (:0
      (Class|BPDirtShellDamageComponent|DamageProp0 _bp_dirtshellmeshnetwork Damage :!class /Figment_S09/Athena/Prototype/Blueprints/Dopey/BP_DirtShellDamageComponent.BP_DirtShellDamageComponent_C))
    (:1
      (Class|BPDirtShellDamageComponent|DamageProp1 _bp_dirtshellmeshnetwork Damage :!class /Figment_S09/Athena/Prototype/Blueprints/Dopey/BP_DirtShellDamageComponent.BP_DirtShellDamageComponent_C))
    (:2
      (Class|BPDirtShellDamageComponent|DamageProp2 _bp_dirtshellmeshnetwork Damage :!class /Figment_S09/Athena/Prototype/Blueprints/Dopey/BP_DirtShellDamageComponent.BP_DirtShellDamageComponent_C))
    (:3
      (Class|BPDirtShellDamageComponent|DamageProp3 _bp_dirtshellmeshnetwork Damage :!class /Figment_S09/Athena/Prototype/Blueprints/Dopey/BP_DirtShellDamageComponent.BP_DirtShellDamageComponent_C))
    (:4
      (Class|BPDirtShellDamageComponent|DamageProp4 _bp_dirtshellmeshnetwork Damage :!class /Figment_S09/Athena/Prototype/Blueprints/Dopey/BP_DirtShellDamageComponent.BP_DirtShellDamageComponent_C))
    (:5
      (Class|BPDirtShellDamageComponent|DamageProp5 _bp_dirtshellmeshnetwork Damage :!class /Figment_S09/Athena/Prototype/Blueprints/Dopey/BP_DirtShellDamageComponent.BP_DirtShellDamageComponent_C))
    (:6
      (Class|BPDirtShellDamageComponent|DamageProp6 _bp_dirtshellmeshnetwork Damage :!class /Figment_S09/Athena/Prototype/Blueprints/Dopey/BP_DirtShellDamageComponent.BP_DirtShellDamageComponent_C))
    (:7
      (Class|BPDirtShellDamageComponent|DamageProp7 _bp_dirtshellmeshnetwork Damage :!class /Figment_S09/Athena/Prototype/Blueprints/Dopey/BP_DirtShellDamageComponent.BP_DirtShellDamageComponent_C))
    (:8
      (Class|BPDirtShellDamageComponent|DamageProp8 _bp_dirtshellmeshnetwork Damage :!class /Figment_S09/Athena/Prototype/Blueprints/Dopey/BP_DirtShellDamageComponent.BP_DirtShellDamageComponent_C))
    (:9
      (Class|BPDirtShellDamageComponent|DamageProp9 _bp_dirtshellmeshnetwork Damage :!class /Figment_S09/Athena/Prototype/Blueprints/Dopey/BP_DirtShellDamageComponent.BP_DirtShellDamageComponent_C))
    (:10
      (Class|BPDirtShellDamageComponent|DamageProp10 _bp_dirtshellmeshnetwork Damage :!class /Figment_S09/Athena/Prototype/Blueprints/Dopey/BP_DirtShellDamageComponent.BP_DirtShellDamageComponent_C))
    (:11
      (Class|BPDirtShellDamageComponent|DamageProp11 _bp_dirtshellmeshnetwork Damage :!class /Figment_S09/Athena/Prototype/Blueprints/Dopey/BP_DirtShellDamageComponent.BP_DirtShellDamageComponent_C))
    (:12
      (Class|BPDirtShellDamageComponent|DamageProp12 _bp_dirtshellmeshnetwork Damage :!class /Figment_S09/Athena/Prototype/Blueprints/Dopey/BP_DirtShellDamageComponent.BP_DirtShellDamageComponent_C))
    (:13
      (Class|BPDirtShellDamageComponent|DamageProp13 _bp_dirtshellmeshnetwork Damage :!class /Figment_S09/Athena/Prototype/Blueprints/Dopey/BP_DirtShellDamageComponent.BP_DirtShellDamageComponent_C))
    (:14
      (Class|BPDirtShellDamageComponent|DamageProp14 _bp_dirtshellmeshnetwork Damage :!class /Figment_S09/Athena/Prototype/Blueprints/Dopey/BP_DirtShellDamageComponent.BP_DirtShellDamageComponent_C))
    (:15
      (Class|BPDirtShellDamageComponent|DamageProp15 _bp_dirtshellmeshnetwork Damage :!class /Figment_S09/Athena/Prototype/Blueprints/Dopey/BP_DirtShellDamageComponent.BP_DirtShellDamageComponent_C))
    (:16
      (Class|BPDirtShellDamageComponent|DamageProp16 _bp_dirtshellmeshnetwork Damage :!class /Figment_S09/Athena/Prototype/Blueprints/Dopey/BP_DirtShellDamageComponent.BP_DirtShellDamageComponent_C))
    (:17
      (Class|BPDirtShellDamageComponent|DamageProp17 _bp_dirtshellmeshnetwork Damage :!class /Figment_S09/Athena/Prototype/Blueprints/Dopey/BP_DirtShellDamageComponent.BP_DirtShellDamageComponent_C))
    (:18
      (Class|BPDirtShellDamageComponent|DamageProp18 _bp_dirtshellmeshnetwork Damage :!class /Figment_S09/Athena/Prototype/Blueprints/Dopey/BP_DirtShellDamageComponent.BP_DirtShellDamageComponent_C))
    (:19
      (Class|BPDirtShellDamageComponent|DamageProp19 _bp_dirtshellmeshnetwork Damage :!class /Figment_S09/Athena/Prototype/Blueprints/Dopey/BP_DirtShellDamageComponent.BP_DirtShellDamageComponent_C))
    (:20
      (Class|BPDirtShellDamageComponent|DamageProp20 _bp_dirtshellmeshnetwork Damage :!class /Figment_S09/Athena/Prototype/Blueprints/Dopey/BP_DirtShellDamageComponent.BP_DirtShellDamageComponent_C))
    (:21
      (Class|BPDirtShellDamageComponent|DamageProp21 _bp_dirtshellmeshnetwork Damage :!class /Figment_S09/Athena/Prototype/Blueprints/Dopey/BP_DirtShellDamageComponent.BP_DirtShellDamageComponent_C))
    (:22
      (Class|BPDirtShellDamageComponent|DamageProp22 _bp_dirtshellmeshnetwork Damage :!class /Figment_S09/Athena/Prototype/Blueprints/Dopey/BP_DirtShellDamageComponent.BP_DirtShellDamageComponent_C))
    (:23
      (Class|BPDirtShellDamageComponent|DamageProp23 _bp_dirtshellmeshnetwork Damage :!class /Figment_S09/Athena/Prototype/Blueprints/Dopey/BP_DirtShellDamageComponent.BP_DirtShellDamageComponent_C))
    (:24
      (Class|BPDirtShellDamageComponent|DamageProp24 _bp_dirtshellmeshnetwork Damage :!class /Figment_S09/Athena/Prototype/Blueprints/Dopey/BP_DirtShellDamageComponent.BP_DirtShellDamageComponent_C))
    (:25
      (Class|BPDirtShellDamageComponent|DamageProp25 _bp_dirtshellmeshnetwork Damage :!class /Figment_S09/Athena/Prototype/Blueprints/Dopey/BP_DirtShellDamageComponent.BP_DirtShellDamageComponent_C))
    (:26
      (Class|BPDirtShellDamageComponent|DamageProp26 _bp_dirtshellmeshnetwork Damage :!class /Figment_S09/Athena/Prototype/Blueprints/Dopey/BP_DirtShellDamageComponent.BP_DirtShellDamageComponent_C))))

(event Custom|WinnerCalled)

(event Custom|ForceVote
  (Variables|Default|SetVoteWinner 0))

(event Custom|ResetWinnerDoonce
  ;[1 1 1] Root node only
  (Utilities|FlowControl|DoOnce false
    (:Completed
      (Variables|Default|SetVoteWinner (stale Custom|UpdateHealth DirtIndex)))))

(event Custom|PropWasDestroyed)

(event Custom|UpdateHealth
  (params (Damage :type Integer) (DirtIndex :type Byte))
  (bind _returnvalue (Math|Conversions|ToInteger(Byte) DirtIndex))
  (bind _returnvalue_2 (Math|Conversions|ToInteger64(Integer) Damage))
  (bind _prophealth (Variables|Default|GetPropHealth))
  (bind _output (Utilities|Array|Get(aref) _prophealth _returnvalue))
  (bind _returnvalue_3 (- _output _returnvalue_2))
  (bind _prophealth_2 (Variables|Default|GetPropHealth))
  ;[1 1 1] Root node only
  (if (<= _returnvalue_3 0)
    (Utilities|FlowControl|DoOnce false
      (:Completed
        (Variables|Default|SetVoteWinner DirtIndex)))
    (else
      (Utilities|Array|SetArrayElem _prophealth_2 _returnvalue _returnvalue_3)
      (Variables|Default|SetPropHealth _prophealth_2)
      ;[1 1 1] Persist the current health to the mesh service
      (bind _structout (Utilities|Struct|SetmembersinBPDirtShellStruct (Variables|Default|GetDirtShellMetaData) (Variables|Default|GetPropHealth)))
      (MeshNetworkSubsystem|SetMetaDatawithKey (GameInstanceSubsystems|GetMeshNetworkSubsystem) (Variables|Default|GetMetaDataKey) _structout))))

(event Custom|PropDamageRecieved
  (params (Damage :type Integer) (DirtProp :type BP%20Dirt%20Shell%20Prop%20Object%20Reference))
  (bind _dirtshellprops (Variables|Default|GetDirtShellProps))
  (bind _returnvalue (Utilities|Array|FindItem _dirtshellprops DirtProp))
  ;[1 1 1] Game server- Send damage up through mesh
  (if (!= _returnvalue -1)
    (CallFunction|SelectServerDamage (Math|Float|Truncate (* Damage (Variables|Default|GetDamageMultiplier))) _returnvalue)))

(event Custom|RootSetHealth
  (params (DirtIndex :type Byte) (TotalHealth :type Integer64))
  (bind _prophealth (Variables|Default|GetPropHealth))
  ;[1 1 1] Root node only
  (Utilities|Array|SetArrayElem _prophealth (Math|Conversions|ToInteger(Byte) DirtIndex) TotalHealth)
  (Variables|Default|SetPropHealth _prophealth))

(event Custom|RootSetAllHealthToDefault
  (bind _prophealth (Variables|Default|GetPropHealth))
  (bind _defaulthealthvalue (Variables|Default|GetDefaultHealthValue))
  ;[1 1 1] Root node only
  (for (_ _array_index) _prophealth
    (Utilities|Array|SetArrayElem _prophealth _array_index _defaulthealthvalue)
    (Class|BuildingPropBigHealth|SetBigHealthMax (Utilities|Array|Get(acopy) (Variables|Default|GetDirtShellProps) _array_index) _defaulthealthvalue))
  (Variables|Default|SetPropHealth _prophealth)
  (bind _structout (Utilities|Struct|SetmembersinBPDirtShellStruct (Variables|Default|GetDirtShellMetaData) (Variables|Default|GetPropHealth)))
  (MeshNetworkSubsystem|SetMetaDatawithKey (GameInstanceSubsystems|GetMeshNetworkSubsystem) (Variables|Default|GetMetaDataKey) _structout))

(event Custom|RootSetDefaultHealth
  (params (DefaultHealth :type Integer64))
  ;[1 1 1] Root node only
  (CallFunction|UpdateDefaultHealth DefaultHealth)
  (bind _structout (Utilities|Struct|SetmembersinBPDirtShellStruct (Variables|Default|GetDirtShellMetaData) (Variables|Default|GetPropHealth) DefaultHealth))
  (MeshNetworkSubsystem|SetMetaDatawithKey (GameInstanceSubsystems|GetMeshNetworkSubsystem) (Variables|Default|GetMetaDataKey) _structout))

(event Custom|HarvestNearby
  (Utilities|Array|Clear (Variables|Default|GetDirtShellProps))
  (bind _outactors (Actor|GetAllActorsOfClass "/Figment_S09/Athena/Prototype/Blueprints/Dopey/BP_DirtShellProp.BP_DirtShellProp_C"))
  (for _array_element _outactors
    (if (<= (Class|AbilityFunctions|GetDistanceBetweenActors _array_element self) 3e+03)
      (Utilities|Array|Add (Variables|Default|GetDirtShellProps) _array_element))))

(event Custom|RootSetDamageMultiplier
  (params (DamageMultiplier :type Float%20%28double-precision%29))
  ;[1 1 1] Root node only
  (bind _output_get (Variables|Default|SetDamageMultiplier DamageMultiplier))
  (bind _structout (Utilities|Struct|SetmembersinBPDirtShellStruct (Variables|Default|GetDirtShellMetaData) _output_get))
  (MeshNetworkSubsystem|SetMetaDatawithKey (GameInstanceSubsystems|GetMeshNetworkSubsystem) (Variables|Default|GetMetaDataKey) _structout))

(event Custom|MeshMetaDataUpdated
  (bind _dirtshellmetadata (Variables|Default|GetDirtShellMetaData))
  (bind (_currenthealth_4_80275d8141fa71e07fa6cfb8ffa9581c _defaulthealth_9_88fbddf2491ddaa24c8dbbac77dcbdc7 _damagemultiplier_12_d0a02b8a4428f3a6a0d918b13c8d3bdb) (Utilities|Struct|BreakBPDirtShellStruct _dirtshellmetadata))
  ;[1 1 1] Restore initial health value if the root was restarted
  (bind _returnvalue (MeshNetworkSubsystem|GetMetaDatawithKey (GameInstanceSubsystems|GetMeshNetworkSubsystem) (Variables|Default|GetMetaDataKey) _dirtshellmetadata))
  (if _returnvalue
    (Variables|Default|SetDefaultHealthValue _defaulthealth_9_88fbddf2491ddaa24c8dbbac77dcbdc7)
    (Variables|Default|SetDamageMultiplier _damagemultiplier_12_d0a02b8a4428f3a6a0d918b13c8d3bdb)
    (for (_array_element _array_index) _currenthealth_4_80275d8141fa71e07fa6cfb8ffa9581c
      (CallFunction|RootSetHealth (Math|Conversions|ToByte(Integer) _array_index) (Math|Integer64|Max(Integer64) _array_element)))))

(event EventBeginPlay
  (Networking|SwitchHasAuthority
    (:Authority
      (Utilities|FlowControl|Sequence
        (:then_0
          ;[1 1 1] Root / Edge Nodes
          (bind _nodetype (Class|FortAsyncActionMeshNetworkReady|MeshNetworkReadyAsync)
            (:OnReady
              (switch Utilities|FlowControl|Switch|SwitchonEMeshNetworkNodeType _nodetype
                (:Root
                  (Utilities|FlowControl|Sequence
                    (:then_0
                      (Default|BindEventtoOnDamageAdded (Variables|Default|GetBP_DirtShellMeshNetwork) (EventDispatchers|CreateEvent :!event Custom|UpdateHealth)))
                    (:then_1
                      (MeshNetworkSubsystem|BindEventtoOnMeshMetaDataUpdated (GameInstanceSubsystems|GetMeshNetworkSubsystem) (EventDispatchers|CreateEvent :!event Custom|MeshMetaDataUpdated))
                      (CallFunction|MeshMetaDataUpdated))
                    (:then_2)))
                (:Edge
                  (Default|BindEventtoOnDamageRecieved (Variables|Default|GetDirtShellProps) (EventDispatchers|CreateEvent :!event Custom|PropDamageRecieved)))))))
        (:then_1
          ;[1 1 1] Hotfix to make sure it stays destroyed
          (bind _asfort_game_state_athena (Utilities|Casting|CastToFortGameStateAthena (Game|GetGameState)))
          (if (not (Variables|/Script/FortniteGame.FortGameStateAthena|GetbPlaylistDataIsLoaded _asfort_game_state_athena))
            (Playlist|BindEventtoOnCurrentPlaylistLoaded _asfort_game_state_athena (delegate Custom|ContinueSetup))))))))

(event Custom|UpdateDefaultHealth
  (params (NewDefaultHealth :type Integer64))
  (Variables|Default|SetNewDefaultHealth NewDefaultHealth)
  (for (_array_element _array_index) (Variables|Default|GetPropHealth)
    (if (Math|Integer64|Equal(Integer64) _array_element (Variables|Default|GetDefaultHealthValue))
      (Utilities|Array|SetArrayElem (Variables|Default|GetPropHealth) _array_index (Variables|Default|GetNewDefaultHealth))))
  (Variables|Default|SetDefaultHealthValue (Variables|Default|GetNewDefaultHealth))
  (Variables|Default|SetPropHealth (Variables|Default|GetPropHealth)))

(event Collision|EventActorBeginOverlap
  (params (OtherActor :type Actor%20Object%20Reference)))

(event Custom|TestDirt
  ;[1 1 1] Root / Edge Nodes
  (Variables|Default|SetNoMeshDebug true)
  (Default|BindEventtoOnDamageRecieved (Variables|Default|GetDirtShellProps) (EventDispatchers|CreateEvent :!event Custom|PropDamageRecieved)))

(event EventTick
  (params (DeltaSeconds :type Float%20%28single-precision%29)))

(event Custom|ContinueSetup
  (params (PlaylistName :type Name) (PlaylistContextTags :type Gameplay%20Tag%20Container%20Structure%20%28by%20ref%29)))

(fn OnRep_DefaultHealthValue
  (if (not (Variables|Default|GetNoMeshDebug))
    (switch Utilities|FlowControl|Switch|SwitchonEMeshNetworkNodeType (MeshNetworking|GetMeshNetworkNodeType (Variables|Default|GetBP_DirtShellMeshNetwork))
      (:Edge
        _)))
  (for (_array_element _array_index) (Variables|Default|GetDirtShellProps)
    (Utilities|IsValid _array_element
      (:Is%20Valid
        (Class|BuildingPropBigHealth|SetBigHealthMax _array_element (Variables|Default|GetDefaultHealthValue))
        (Class|BuildingPropBigHealth|SetBigHealth _array_element (Utilities|Array|Get(acopy) (Variables|Default|GetPropHealth) _array_index) false)))))

(fn OnRep_PropHealth
  (bind _prophealth (Variables|Default|GetPropHealth))
  (if (not (Variables|Default|GetNoMeshDebug))
    (switch Utilities|FlowControl|Switch|SwitchonEMeshNetworkNodeType (MeshNetworking|GetMeshNetworkNodeType (Variables|Default|GetBP_DirtShellMeshNetwork))
      (:Edge
        _)))
  (for (_array_element _array_index) (Variables|Default|GetDirtShellProps)
    (Utilities|IsValid _array_element
      (:Is%20Valid
        (if (Math|Integer64|NotEqual(Integer64) (Variables|/Script/FortniteGame.BuildingPropBigHealth|GetBigHealth _array_element) (Utilities|Array|Get(acopy) _prophealth _array_index))
          (Class|BPDirtShellProp|EdgeServerApplyDamage _array_element (Utilities|Array|Get(acopy) _prophealth _array_index) :!class /Figment_S09/Athena/Prototype/Blueprints/Dopey/BP_DirtShellProp.BP_DirtShellProp_C)))))
  (if (not (Variables|Default|GetFirstSetupComplete))
    (Variables|Default|SetFirstSetupComplete true)
    (Development|PrintString "first set up" true true "(R=0.000000,G=0.660000,B=1.000000,A=1.000000)" 6e+01)))

(fn OnRep_Winner
  (switch Utilities|FlowControl|Switch|SwitchonEMeshNetworkNodeType (MeshNetworkSubsystem|GetMeshNetworkNodeType (GameInstanceSubsystems|GetMeshNetworkSubsystem))
    (:Root
      _)
    (:Edge
      _))
  (CallFunction|WinnerCalled))

(fn ConstructionScript
  (bind _prophealth (Variables|Default|GetPropHealth))
  (Utilities|Array|Clear _prophealth)
  (for (_ _array_index) (Variables|Default|GetDirtShellProps)
    (Utilities|Array|SetArrayElem _prophealth _array_index (Variables|Default|GetDefaultHealthValue) true)))
